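; Rocket Drill weapon item that manages idle effects and fuel display for the player holding it.
; Riding events toggle the riding flag and looping idle GameplayCue, and stopping also reloads the weapon.
; Nested tag-query waits hide the idle Niagara effect while the vehicle tag query matches.
; Tick feeds the pawn's Fuel attribute into that effect, whereas the normalized-fuel getter uses magazine ammo.
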(event Custom|StopRidingRocketDrill
  (GameplayCue|AddGameplayCue self (Variables|Default|GetLoopingIdleCue))
  (Variables|Default|SetisRiding false)
  (FortWeapon|Reload))

(event Custom|RidingRocketDrill
  (GameplayCue|RemoveGameplayCue self (Variables|Default|GetLoopingIdleCue))
  (Variables|Default|SetisRiding true))

(event Custom|BoostStateChanged
  (params (bIsBoosting :type Boolean)))

(event EventBeginPlay
  (bind _asfort_player_pawn (Utilities|Casting|CastToFortPlayerPawn (Game|GetInstigator)))
  (Variables|Default|SetAs%20Fort%20Player%20Pawn _asfort_player_pawn)
  (Ability|Tasks|WaitGameplayTagQueryOnActor _asfort_player_pawn (Variables|Default|GetVehicleTag)
    (:Triggered
      (Development|SetHiddeninGame (Variables|Default|GetNS_idle) true false)
      (Ability|Tasks|WaitGameplayTagQueryOnActor _asfort_player_pawn (Variables|Default|GetVehicleTag) "WhenFalse"
        (:Triggered
          (Development|SetHiddeninGame (Variables|Default|GetNS_idle) false false))))))

(event EventTick
  (params (DeltaSeconds :type Float%20%28single-precision%29))
  (bind (_bsuccessfullyfoundattribute _returnvalue) (Ability|Attribute|GetFloatAttribute (Variables|Default|GetAs%20Fort%20Player%20Pawn) "(AttributeName=\"Fuel\",Attribute=/Script/RocketDrillGameplayRuntime.RocketDrillAttributeSet:Fuel,AttributeOwner=\"/Script/CoreUObject.Class'/Script/RocketDrillGameplayRuntime.RocketDrillAttributeSet'\")"))
  (bind _as_fort_player_pawn (Variables|Default|GetAs%20Fort%20Player%20Pawn))
  (Utilities|IsValid _as_fort_player_pawn
    (:Is%20Valid
      (Niagara|SetNiagaraVariable(Float) (Variables|Default|GetNS_idle) "FuelAmount" (Math|Float|NormalizetoRange _returnvalue 0.0 1e+02)))))

(fn CheatToggleAlwaysDestroy
  (Variables|Default|SetCheatAlwaysDestroy (not (Variables|Default|GetCheatAlwaysDestroy))))

(fn GetNormalizedFuelRemaining
  (return (/ (FortWeapon|GetMagazineAmmoCount) (FortWeapon|GetBulletsPerClip))))

(fn GetOutOfAmmoTextOverride
  (return "NSLOCTEXT(\"[BC12AAD80160A615F82A8397C5DFB9F3]\", \"58D7C576454267C585492EBE21050534\", \"No Fuel\")"))

(fn ConstructionScript
  (|CallParentFunction))
