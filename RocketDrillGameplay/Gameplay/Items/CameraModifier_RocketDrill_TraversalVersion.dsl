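; Camera modifier for Rocket Drill traversal that offsets the player's view and widens FOV during movement.
; It anchors the pawn's start location, driving curve-based pull-out and FOV from forward distance travelled.
; Frame-to-frame velocity feeds interpolated side, vertical and lean offsets, with boost-specific vertical tuning and a shooting blend, all scaled by modifier alpha.
; WindDownAndRemove blends out, then removes the modifier unless the drill is active again.
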(event Custom|WindDownAndRemove
  (CameraModifier|DisableModifier false)
  (Utilities|FlowControl|Delay (Variables|CameraModifier|GetAlphaOutTime))
  (if (not (GameplayTags|HasMatchingGameplayTag (Game|Player|GetPlayerPawn (Utilities|Casting|CastToFortPlayerController (Game|Player|GetOwningPlayerController (Variables|CameraModifier|GetCameraOwner)))) "(TagName=\"Gameplay.Action.Player.RocketDrill.Active\")"))
    (Game|Player|RemoveCameraModifier (Variables|CameraModifier|GetCameraOwner) self)))

(fn BlueprintModifyCamera
  (params (DeltaTime :type Float%20%28single-precision%29) (ViewLocation :type Vector :default "0, 0, 0") (ViewRotation :type Rotator :default "0, 0, 0") (FOV :type Float%20%28single-precision%29))
  (bind _anchoredlocation (Variables|Default|GetAnchoredLocation))
  (bind _playerpawn_athena (Variables|Default|GetPlayerPawn_Athena))
  (bind _returnvalue (Transformation|GetActorLocation _playerpawn_athena))
  (bind _returnvalue_2 (- _returnvalue _anchoredlocation))
  (bind _viewrotation (Variables|GetViewRotation))
  (bind _returnvalue_3 (Math|Vector|GetForwardVector _viewrotation))
  (bind _returnvalue_4 (Math|Vector|DotProduct _returnvalue_3 _returnvalue_2))
  (bind _velocity (Variables|Default|GetVelocity))
  (Utilities|FlowControl|Sequence
    (:then_0
      (Utilities|IsValid (Variables|Default|GetPlayerPawn_Athena)
        (:Is%20Not%20Valid
          (bind _asplayer_pawn_athena (Utilities|Casting|CastToPlayerPawn_Athena (CameraModifier|GetViewTarget)))
          (Variables|Default|SetPlayerPawn_Athena _asplayer_pawn_athena)
          (Variables|Default|SetAnchoredLocation (Transformation|GetActorLocation (Variables|Default|GetPlayerPawn_Athena))))))
    (:then_1
      (Variables|Default|SetLeadIn_ForwardOffset (Utilities|UnNormalizeFloat (Math|Curves|GetFloatValue (Variables|Default|GetDistanceTraveledToCameraPullOut) (Utilities|NormalizeFloat _returnvalue_4 0 7.5e+02)) 0 7.5e+02))
      (Variables|Default|SetCurrentFOV (Utilities|RemapFloat _returnvalue_4 "(X=0.000000,Y=3000.000000)" "(X=0.000000,Y=30.000000)" true)))
    (:then_2
      (bind _output_get (Variables|Default|SetVelocity (Math|Vector|SelectVector (Variables|Default|GetVelocity) (/ (- (Transformation|GetActorLocation (Variables|Default|GetPlayerPawn_Athena)) (Variables|Default|GetLastFramePosition_)) (Math|Float|SelectFloat 1.0 (Variables|GetDeltaTime) (Math|Float|NearlyEqual(Float) (Variables|GetDeltaTime)))) (Math|Vector|VectorIsNearlyZero (/ (- (Transformation|GetActorLocation (Variables|Default|GetPlayerPawn_Athena)) (Variables|Default|GetLastFramePosition_)) (Math|Float|SelectFloat 1.0 (Variables|GetDeltaTime) (Math|Float|NearlyEqual(Float) (Variables|GetDeltaTime))))))))
      (LocalVariables|SetDirection (Math|Vector|Normalize _output_get))
      (LocalVariables|SetSpeed (Math|Vector|VectorLength _output_get))
      (Variables|Default|SetLastFramePosition_ (Transformation|GetActorLocation (Variables|Default|GetPlayerPawn_Athena))))
    (:then_3
      (Variables|Default|SetADS_Alpha (Math|Interpolation|FInterpTo (Variables|Default|GetADS_Alpha) (Math|Float|SelectFloat 1.0 0.0 (LocalVariables|GetIsShooting)) (Variables|GetDeltaTime) 4.0))
      (LocalVariables|SetForwardSpeed (Math|Vector|DotProduct _velocity (Math|Vector|GetForwardVector (Variables|GetViewRotation))))
      (LocalVariables|SetRightSpeed (Math|Vector|DotProduct _velocity (Math|Vector|GetRightVector (Variables|GetViewRotation))))
      (LocalVariables|SetUpSpeed (Math|Vector|DotProduct _velocity (Math|Vector|GetUpVector (Variables|GetViewRotation))))
      (LocalVariables|SetCurrentStraightness (Math|Float|Absolute(Float) (Math|Vector|DotProduct (Math|Vector|GetForwardVector (Variables|GetViewRotation)) (LocalVariables|GetDirection))))
      (Variables|Default|SetCurrentForward 0.0)
      (Variables|Default|SetCurrentRight (Math|Interpolation|FInterpTo (Variables|Default|GetCurrentRight) (Utilities|Operators|Multiply (Utilities|RemapFloat (LocalVariables|GetRightSpeed) "(X=-1000.000000,Y=1000.000000)" "(X=-150.000000,Y=150.000000)" true) (Math|Float|SelectFloat 1.3 0.9 (> (Utilities|RemapFloat (LocalVariables|GetRightSpeed) "(X=-1000.000000,Y=1000.000000)" "(X=-150.000000,Y=150.000000)" true) 0)) 1.0 (Utilities|RemapFloat (Math|Float|Absolute(Float) (Variables|Default|GetCurrentForward)) "(X=50.000000,Y=300.000000)" "(X=1.200000,Y=1.500000)" true)) (Variables|GetDeltaTime) 3.0))
      (Variables|Default|SetCurrentUp (Math|Interpolation|FInterpTo (Variables|Default|GetCurrentUp) (+ (Utilities|Operators|Multiply (Utilities|RemapFloat (LocalVariables|GetUpSpeed) "(X=-1000.000000,Y=1000.000000)" "(X=-170.000000,Y=170.000000)" true) (Utilities|RemapFloat (Math|Float|Absolute(Float) (Variables|Default|GetCurrentForward)) "(X=50.000000,Y=400.000000)" "(X=1.000000,Y=1.600000)" true) (Math|Float|SelectFloat (ScalableFloat|GetValueatLevel (Variables|Default|GetSprintingUpOffsetMultiplier)) (ScalableFloat|GetValueatLevel (Variables|Default|GetBaseUpOffsetMultiplier)) (CallFunction|IsBoosting)) (Math|Float|SelectFloat 1.35 1.0 (> (Utilities|RemapFloat (LocalVariables|GetUpSpeed) "(X=-1000.000000,Y=1000.000000)" "(X=-170.000000,Y=170.000000)" true) 0))) (Math|Float|SelectFloat -6e+01 -3e+01 (CallFunction|IsBoosting))) (Variables|GetDeltaTime) 4.0))
      (Variables|Default|SetCurrentLean (Math|Interpolation|FInterpTo (Variables|Default|GetCurrentLean) (Utilities|RemapFloat (* (LocalVariables|GetCurrentStraightness) (LocalVariables|GetLeanRight)) "(X=-1.000000,Y=1.000000)" (Math|Vector2D|MakeVector2D (Utilities|RemapFloat (Math|Float|Absolute(Float) (LocalVariables|GetForwardSpeed)) "(X=800.000000,Y=1800.000000)" "(X=40.000000,Y=50.000000)" true) (* (Utilities|RemapFloat (Math|Float|Absolute(Float) (LocalVariables|GetForwardSpeed)) "(X=800.000000,Y=1800.000000)" "(X=40.000000,Y=50.000000)" true) -2.75)) true) (Variables|GetDeltaTime) 4.5))
      (Variables|Default|SetCurrentOffset (Math|Interpolation|FInterpTo (Variables|Default|GetCurrentForward) (Utilities|RemapFloat (LocalVariables|GetSpeed) "(X=-3500.000000,Y=3500.000000)" "(X=-220.000000,Y=220.000000)" true) (Variables|GetDeltaTime) 4.0))
      (Variables|Default|SetCurrentOffset_Vec (Math|Interpolation|VInterpTo (Variables|Default|GetCurrentOffset_Vec) (Utilities|Operators|Add (* (Math|Vector|GetForwardVector (Variables|GetViewRotation)) (Variables|Default|GetCurrentForward)) (* (Math|Vector|GetRightVector (Variables|GetViewRotation)) (+ (Variables|Default|GetCurrentRight) (* (LocalVariables|GetCurrentStraightness) -9e+01))) (* "0.000000,0.000000,1.000000" (- (Variables|Default|GetCurrentUp) 18.0)) (* (Math|Vector|CrossProduct (LocalVariables|GetDirection) "0.000000,0.000000,1.000000") (Variables|Default|GetCurrentLean))) (Variables|GetDeltaTime) 1e+01)))
    (:then_4
      (return (- (Variables|GetViewLocation) (+ (* (Utilities|Operators|Add (* (Variables|Default|GetCurrentOffset_Vec) (- 1.0 (Variables|Default|GetADS_Alpha))) (Utilities|Operators|Multiply (Math|Vector|GetForwardVector (Variables|GetViewRotation)) (- 1.0 (Variables|Default|GetADS_Alpha)) -3e+01) (* (Math|Vector|GetUpVector (Variables|GetViewRotation)) 3e+01) (* (Math|Vector|GetRightVector (Variables|GetViewRotation)) -1e+01)) (Utilities|CalculateAlpha)) (* (* (Math|Vector|GetForwardVector (Variables|GetViewRotation)) (Variables|Default|GetLeadIn_ForwardOffset)) (Utilities|CalculateAlpha)))) (Variables|GetViewRotation) (+ (Variables|GetFOV) (* (Variables|Default|GetCurrentFOV) (Utilities|CalculateAlpha)))))))

(fn IsBoosting
  (return (GameplayTags|HasMatchingGameplayTag (Variables|Default|GetPlayerPawn_Athena) "(TagName=\"Granted.Athena.RocketDrill.Boosting\")")))

(fn SetStartingLocation
  (params (StartingLocation :type Vector))
  (Variables|Default|SetStartingLocation StartingLocation))
