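; Test camera modifier in the Rocket Drill item gameplay folder that animates view offset and FOV for a Fortnite player pawn.
; BlueprintModifyCamera caches pawn and start time, then drives FOV and directional offsets through two timed transitions, scaled by modifier alpha.
; After 1.25 seconds it anchors the pawn's location and counters later forward movement with a curve-weighted offset, creating camera lag.
; WindDownAndRemove disables the modifier, waits out the blend, then removes it.
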(event Custom|WindDownAndRemove
  (CameraModifier|DisableModifier false)
  (Utilities|FlowControl|Delay (Variables|CameraModifier|GetAlphaOutTime))
  (Game|Player|RemoveCameraModifier (Variables|CameraModifier|GetCameraOwner) self))

(fn BlueprintModifyCamera
  (params (DeltaTime :type Float%20%28single-precision%29) (ViewLocation :type Vector :default "0, 0, 0") (ViewRotation :type Rotator :default "0, 0, 0") (FOV :type Float%20%28single-precision%29))
  (Utilities|FlowControl|Sequence
    (:then_0
      ; init
      (Utilities|IsValid (Variables|Default|GetPlayerPawn_Athena)
        (:Is%20Not%20Valid
          (bind _asplayer_pawn_athena (Utilities|Casting|CastToPlayerPawn_Athena (Game|Player|GetPlayerPawn (Utilities|Casting|CastToFortPlayerController (Game|Player|GetOwningPlayerController (Variables|CameraModifier|GetCameraOwner))))))
          (Variables|Default|SetPlayerPawn_Athena _asplayer_pawn_athena)
          (Variables|Default|SetStartTime (Utilities|Time|GetGameTimeinSeconds)))))
    (:then_1
      ; FOV interp
      (Variables|Default|SetTargetFOVModifier (+ (* (Variables|Default|GetFOV1) (- (Utilities|NormalizeFloat (- (Utilities|Time|GetGameTimeinSeconds) (Variables|Default|GetStartTime)) 0 (Variables|Default|GetFirstTransitionDuration) true) (Utilities|NormalizeFloat (- (Utilities|Time|GetGameTimeinSeconds) (Variables|Default|GetStartTime)) (Variables|Default|GetFirstTransitionDuration) (+ (Variables|Default|GetFirstTransitionDuration) (Variables|Default|GetSecondTransitionDuration)) true))) (* (Variables|Default|GetFOV2) (Utilities|NormalizeFloat (- (Utilities|Time|GetGameTimeinSeconds) (Variables|Default|GetStartTime)) (Variables|Default|GetFirstTransitionDuration) (+ (Variables|Default|GetFirstTransitionDuration) (Variables|Default|GetSecondTransitionDuration)) true)))))
    (:then_2
      (Variables|Default|SetTargetViewOffset (Utilities|Operators|Add (+ (* (Math|Vector|GetForwardVector (Variables|GetViewRotation)) (* (Variables|Default|GetForwardBackwardOffset) (- (Utilities|NormalizeFloat (- (Utilities|Time|GetGameTimeinSeconds) (Variables|Default|GetStartTime)) 0 (Variables|Default|GetFirstTransitionDuration) true) (Utilities|NormalizeFloat (- (Utilities|Time|GetGameTimeinSeconds) (Variables|Default|GetStartTime)) (Variables|Default|GetFirstTransitionDuration) (+ (Variables|Default|GetFirstTransitionDuration) (Variables|Default|GetSecondTransitionDuration)) true)))) (* (Math|Vector|GetForwardVector (Variables|GetViewRotation)) (* 0 (Utilities|NormalizeFloat (- (Utilities|Time|GetGameTimeinSeconds) (Variables|Default|GetStartTime)) (Variables|Default|GetFirstTransitionDuration) (+ (Variables|Default|GetFirstTransitionDuration) (Variables|Default|GetSecondTransitionDuration)) true)))) (* "0.000000,0.000000,1.000000" (* (Utilities|NormalizeFloat (- (Utilities|Time|GetGameTimeinSeconds) (Variables|Default|GetStartTime)) 0 (Variables|Default|GetFirstTransitionDuration) true) (Variables|Default|GetUpDownOffset))) (* (Math|Vector|GetForwardVector (Variables|GetViewRotation)) (Variables|Default|GetLeadIn_ForwardOffset)) (* (Math|Vector|GetRightVector (Variables|GetViewRotation)) (* (Utilities|NormalizeFloat (- (Utilities|Time|GetGameTimeinSeconds) (Variables|Default|GetStartTime)) 0 (Variables|Default|GetFirstTransitionDuration) true) (Variables|Default|GetLeftRightOffset)))))
      (if (>= (Utilities|NormalizeFloat (- (Utilities|Time|GetGameTimeinSeconds) (Variables|Default|GetStartTime)) 0 1.25 true) 1.0)
        (if (Utilities|Operators|Equal(==) :A (Variables|Default|GetAnchored%20Location))
          (Variables|Default|SetAnchored%20Location (Transformation|GetActorLocation (Variables|Default|GetPlayerPawn_Athena)))
          (else
            (Variables|Default|SetLeadIn_ForwardOffset (Utilities|Operators|Multiply (Math|Vector|DotProduct (Math|Vector|GetForwardVector (Variables|GetViewRotation)) (- (Transformation|GetActorLocation (Variables|Default|GetPlayerPawn_Athena)) (Variables|Default|GetAnchored%20Location))) (Math|Curves|GetFloatValue (Variables|Default|GetLeadInOffsetCurve) (Utilities|NormalizeFloat (- (Utilities|Time|GetGameTimeinSeconds) (Variables|Default|GetStartTime)) 1.25 (+ 1.25 0.75) true)) -0.9))))))
    (:then_3
      (return (+ ViewLocation (* (Variables|Default|GetTargetViewOffset) (Variables|CameraModifier|GetAlpha))) ViewRotation (+ FOV (* (Variables|Default|GetTargetFOVModifier) (Variables|CameraModifier|GetAlpha)))))))
